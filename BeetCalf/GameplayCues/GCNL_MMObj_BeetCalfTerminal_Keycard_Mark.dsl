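; Looping gameplay cue that marks the closest keycard-accepting BeetCalf radio tower terminal for the target player.
; Overrides the parent's terminal search to pick the nearest uncaptured terminal from the tower game state component's list.
; Marking briefly enables markability around the parent call, then binds to the terminal's state changes.
; The terminal is unmarked when it stops accepting keycards, and the delegate is unbound on unmark or cue removal.
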(event EventOnRemoval(EitherVFX-preferthisversion)
  (params (MyTarget :type Actor%20Object%20Reference) (Parameters :type Gameplay%20Cue%20Parameters%20Structure%20%28by%20ref%29) (ParticleComponents :type Array%20of%20FXSystem%20Component%20Object%20References) (AudioComponents :type Array%20of%20Audio%20Component%20Object%20References) (BurstCameraShakeInstance :type Camera%20Shake%20Base%20Object%20Reference) (BurstDecalInstance :type Decal%20Actor%20Object%20Reference))
  ;[0.1667 0.1076 0.02458] OnRemoval: Unbind delegates.
  (|CallParentFunction MyTarget Parameters ParticleComponents AudioComponents BurstCameraShakeInstance BurstDecalInstance)
  (bind _asbga_mmobj_radio_tower_terminal (Utilities|Casting|CastToBGA_MMObj_RadioTowerTerminal (Variables|Default|GetClosestLockedTerminal)))
  (Default|UnbindEventfromOnTerminalStateChanged _asbga_mmobj_radio_tower_terminal (EventDispatchers|CreateEvent :!event HandleClosestTerminalStateChanged)))

(fn FindClosestKeycardAcceptingTerminal
  (bind _targetplayerpawn (Variables|Default|GetTargetPlayerPawn)
    (:then
      (Utilities|IsValid (Fortnite|Athena|GameState|GetGameStateAthena _targetplayerpawn)
        (:Is%20Valid
          (Utilities|IsValid (Actor|GetComponentbyClass (Fortnite|Athena|GameState|GetGameStateAthena _targetplayerpawn) "/BeetCalf/Gameplay/B_MMObj_BeetCalfTowerGameStateComponent.B_MMObj_BeetCalfTowerGameStateComponent_C")
            (:Is%20Valid
              (bind _terminals (General|GetRadioTowerTerminals (Actor|GetComponentbyClass (Fortnite|Athena|GameState|GetGameStateAthena _targetplayerpawn) "/BeetCalf/Gameplay/B_MMObj_BeetCalfTowerGameStateComponent.B_MMObj_BeetCalfTowerGameStateComponent_C") :!class /BeetCalf/Gameplay/B_MMObj_BeetCalfTowerGameStateComponent.B_MMObj_BeetCalfTowerGameStateComponent_C))
              (for _array_element _terminals
                (Utilities|IsValid _array_element
                  (:Is%20Valid
                    (if (not (Variables|/BeetCalf/Gameplay/BGA_MMObj_BeetCalfTerminal.BGA_MMObj_BeetCalfTerminal_C|GetCaptured _array_element))
                      (Utilities|Array|Add (LocalVariables|GetKeycardAcceptingTerminals) _array_element)))))
              (return (Fortnite|GetClosestActorfromArray (Variables|Default|GetTargetPlayerPawn) (LocalVariables|GetKeycardAcceptingTerminals))))))))
    (:else)))

(fn HandleClosestTerminalStateChanged
  (params (Terminal :type BGA%20MMObj%20Radio%20Tower%20Terminal%20Object%20Reference) (NewState :type MMObjRadioTowerState%20Enum :default "NewEnumerator5"))
  ;[0.1042 0.05303 0.006571] HandleClosestTerminalStateChanged: If the terminal state changes to one where it isn't accepting keycards anymore then unmark that terminal.
  (Utilities|IsValid Terminal
    (:Is%20Valid
      (if (not (Reward|IsAcceptingKeycard Terminal :!class /MidMatchRadioTowerGameplay/Gameplay/BGA_MMObj_RadioTowerTerminal.BGA_MMObj_RadioTowerTerminal_C))
        (CallFunction|UnmarkClosestTerminal)))))

(fn MarkClosestTerminal
  ;[0.09436 0.1302 0.07149] MarkClosestTerminal: Setup a delegate that fires whenever the terminal's state changes.
  (bind _asbga_mmobj_radio_tower_terminal (Utilities|Casting|CastToBGA_MMObj_RadioTowerTerminal (Variables|Default|GetClosestLockedTerminal)))
  (MarkableActor|SetCanBeMarked _asbga_mmobj_radio_tower_terminal true)
  (|CallParentFunction)
  (MarkableActor|SetCanBeMarked _asbga_mmobj_radio_tower_terminal false)
  (Default|BindEventtoOnTerminalStateChanged _asbga_mmobj_radio_tower_terminal (EventDispatchers|CreateEvent :!event HandleClosestTerminalStateChanged)))

(fn UnmarkClosestTerminal
  ;[0.1712 0.05908 0.1823] UnmarkClosestTermina: Unmark the closest terminal, unbind from knowing when its state changes.
  (|CallParentFunction)
  (bind _asbga_mmobj_radio_tower_terminal (Utilities|Casting|CastToBGA_MMObj_RadioTowerTerminal (Variables|Default|GetClosestLockedTerminal)))
  (Default|UnbindEventfromOnTerminalStateChanged _asbga_mmobj_radio_tower_terminal (EventDispatchers|CreateEvent :!event HandleClosestTerminalStateChanged)))

(fn ConstructionScript
  (|CallParentFunction))
